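# Physics asset for the Fuzzy Bear two-handed pickaxe skeletal mesh, giving its bear hand, fuzzy piece and bow limited secondary motion.
# A handle capsule has no constraint, while the dyn_ bodies (capsules plus one Fuzzy box) hang from dyn_BearHand_root.
# Fingers and thumbs chain off the hand, and four two-segment bow knot chains branch from the root.
# Bow joints get looser swing limits than hand and Fuzzy joints, with tight twist throughout.
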PhysicsAsset /BRCosmetics/Weapons/FORT_Melee/Pickaxe_FuzzyBear/Meshes/SK_Pickaxe_Fuzzy_Bear_Physics
bodies {
 Melee_TwoHanded_Handle [{shapeType=Capsule, length=91.386, radius=2.2734, center=<0.4601 -0.9797 13.3696>, rotation=<rot 0.1897 -90 0.137>}]
 dyn_BearHand [{shapeType=Capsule, length=5, radius=10, center=<5 0 0>, rotation=<rot 90 0 0>}]
 dyn_BearHand_Index_Finger [{shapeType=Capsule, length=10, radius=6.5, center=<8 0 0>, rotation=<rot 90 0 0>}]
 dyn_BearHand_Little_Thumb [{shapeType=Capsule, length=6, radius=5.6, center=<4 0 0>, rotation=<rot 90 0 0>}]
 dyn_BearHand_Thumb [{shapeType=Capsule, length=5.4, radius=7, center=<6 0 0>, rotation=<rot 90 0 0>}]
 dyn_BearHand_root [{shapeType=Capsule, length=0.02, radius=2.8131, center=<-0.0465 0.1919 -0.0786>, rotation=<rot 0 -0.0966 -89.6956>}]
 dyn_Fuzzy [{shapeType=Box, extentX=7.5, extentY=23, extentZ=20, center=<0.5 0 0>}]
 dyn_bow_knot_01_l_1 [{shapeType=Capsule, length=5, radius=3, center=<-2.903 0 0>, rotation=<rot 90 0 0>}]
 dyn_bow_knot_01_l_2 [{shapeType=Capsule, length=3, radius=5, center=<-2.935 0 0>, rotation=<rot 90 0 0>}]
 dyn_bow_knot_01_r_1 [{shapeType=Capsule, length=5, radius=3, center=<2.1666 0 0>, rotation=<rot 90 0 0>}]
 dyn_bow_knot_01_r_2 [{shapeType=Capsule, length=3, radius=5, center=<2.9051 0 0>, rotation=<rot 90 0 0>}]
 dyn_bow_knot_02_l_1 [{shapeType=Capsule, length=5, radius=1, center=<-2.039 0 0>, rotation=<rot 90 0 0>}]
 dyn_bow_knot_02_l_2 [{shapeType=Capsule, length=3, radius=2, rotation=<rot 90 0 0>}]
 dyn_bow_knot_02_r_1 [{shapeType=Capsule, length=5, radius=1, center=<1.8195 0 0>, rotation=<rot 90 0 0>}]
 dyn_bow_knot_02_r_2 [{shapeType=Capsule, length=3, radius=2, rotation=<rot 90 0 0>}]
}
constraints [
 {bone1Name=dyn_BearHand, bone2Name=dyn_BearHand_root, swing1LimitDegrees=15, swing2LimitDegrees=15, twistLimitDegrees=5}
 {bone1Name=dyn_BearHand_Thumb, bone2Name=dyn_BearHand, swing1LimitDegrees=15, swing2LimitDegrees=15, twistLimitDegrees=5}
 {bone1Name=dyn_BearHand_Index_Finger, bone2Name=dyn_BearHand, swing1LimitDegrees=15, swing2LimitDegrees=15, twistLimitDegrees=5}
 {bone1Name=dyn_BearHand_Little_Thumb, bone2Name=dyn_BearHand, swing1LimitDegrees=15, swing2LimitDegrees=15, twistLimitDegrees=5}
 {bone1Name=dyn_Fuzzy, bone2Name=dyn_BearHand_root, swing1LimitDegrees=15, swing2LimitDegrees=15, twistLimitDegrees=5}
 {bone1Name=dyn_bow_knot_02_l_1, bone2Name=dyn_BearHand_root, swing1LimitDegrees=25, swing2LimitDegrees=25, twistLimitDegrees=5}
 {bone1Name=dyn_bow_knot_02_l_2, bone2Name=dyn_bow_knot_02_l_1, swing1LimitDegrees=25, swing2LimitDegrees=25, twistLimitDegrees=5}
 {bone1Name=dyn_bow_knot_02_r_1, bone2Name=dyn_BearHand_root, swing1LimitDegrees=25, swing2LimitDegrees=25, twistLimitDegrees=5}
 {bone1Name=dyn_bow_knot_02_r_2, bone2Name=dyn_bow_knot_02_r_1, swing1LimitDegrees=25, swing2LimitDegrees=25, twistLimitDegrees=5}
 {bone1Name=dyn_bow_knot_01_l_1, bone2Name=dyn_BearHand_root, swing1LimitDegrees=25, swing2LimitDegrees=25, twistLimitDegrees=5}
 {bone1Name=dyn_bow_knot_01_l_2, bone2Name=dyn_bow_knot_01_l_1, swing1LimitDegrees=25, swing2LimitDegrees=25, twistLimitDegrees=5}
 {bone1Name=dyn_bow_knot_01_r_1, bone2Name=dyn_BearHand_root, swing1LimitDegrees=25, swing2LimitDegrees=25, twistLimitDegrees=5}
 {bone1Name=dyn_bow_knot_01_r_2, bone2Name=dyn_bow_knot_01_r_1, swing1LimitDegrees=25, swing2LimitDegrees=25, twistLimitDegrees=5}
]
